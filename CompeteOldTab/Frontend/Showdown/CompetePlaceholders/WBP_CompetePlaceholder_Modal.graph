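# Placeholder modal for the Compete Showdown frontend, telling the player Battle Royale content is missing and must be downloaded from the lobby.
# A CommonActivatableWidget lays out title, body text and a "Go to lobby" button over a dark UIKit backplate.
# Clicking the button casts the current UI state widget to the Athena frontend state, navigates to lobby pedestals, then fires an element-interacted analytics event.
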WidgetBlueprint /CompeteOldTab/Frontend/Showdown/CompetePlaceholders/WBP_CompetePlaceholder_Modal
graphs [EventGraph {nodes [
   "On Clicked (CTA_Button)"!() -> (OutputDelegate, then, Button:Common Button Base Object Reference) {is=title K2Node_ComponentBoundEvent}
   > n1 = GetCurrentStateWidget!(
    execute
    PlayerController:Player Controller Object Reference=GetPlayerController!(PlayerIndex:int=0) -> (ReturnValue:Player Controller Object Reference)
   ) -> (then, ReturnValue:Common Activatable Widget Object Reference)
   > Is Valid(InputObject=@n1)
   > n2 = Cast To AthenaUIStateWidget_Frontend!(execute, Object=@n1) -> (then, CastFailed, AsAthena UIState Widget Frontend:Athena UIState Widget Frontend Object Reference) {is=title K2Node_DynamicCast}
   > NavigateToLobbyPedestals(self:Athena UIState Widget Frontend Object Reference=@n2)
   > n3 = Cast To FortPlayerController(Object=GetOwningPlayer())
   > "Fire Event Element Interacted (Non-Localized)"!(
    execute
    FortPC:Fort Player Controller Object Reference=@n3
    ElementName:String=Button_GoToLobby
    AdditionalInfo:Flow Analytics Additional Info Structure
   ) -> (then)
   Event PreConstruct()
   Event Construct()
  ]}]
variables []
widgets {
 info {inheritedWidgetCount=0, parentClass=&CommonActivatableWidget, rootWidgetClass=&Overlay, widgetCount=6}
 widgets [
  Overlay_0 {widgetClassPath=&Overlay, widget=Overlay_0}
  WBP_UIKit_Backplate {
   widgetClassPath=&WBP_UIKit_Backplate
   bIsVariable=true
   parent=Overlay_0
   widget=WBP_UIKit_Backplate
   properties {brightness=Dark}
   slot {class=OverlaySlot, properties {horizontalAlignment=HAlign_Fill, verticalAlignment=VAlign_Fill}}
  }
  Root {
   widgetClassPath=&GridPanel
   parent=Overlay_0
   widget=Root
   properties {
    navigation=&self:WidgetTree.Root.WidgetNavigation_0
    columnFill [1]
    rowFill [0,0,0]
   }
   slot {class=OverlaySlot, properties {padding=<ltrb 80 80 80 48>}}
  }
  Text_Title {
   widgetClassPath=&CommonTextBlock
   parent=Root
   widget=Text_Title
   properties {
    text="WAIT, SOMETHING'S MISSING"
    textTransformPolicy=ToUpper
    wrapTextAt=1000
    font {
     bForceMonospaced=false
     bMaterialIsStencil=false
     fontMaterial=null
     fontObject=&HeadingNow
     letterSpacing=0
     monospacedWidth=1
     size=25.5
     skewAmount=0.3
     typefaceFontName=76 Bold
     outlineSettings {
      bApplyOutlineToDropShadows=false
      bMiteredCorners=false
      bSeparateFillAlpha=false
      outlineMaterial=null
      outlineSize=0
      outlineColor=<rgba 1 1 1 0>
     }
    }
    shadowOffset=<0 0>
   }
   slot {class=GridSlot, properties {horizontalAlignment=HAlign_Left, verticalAlignment=VAlign_Top}}
  }
  Text_Body {
   widgetClassPath=&CommonTextBlock
   parent=Root
   widget=Text_Body
   properties {
    style=&TextStyle-Base_C
    text="You need the Battle Royal content in order to proceed.\r\n\r\nStart the download in the lobby to gain access to the complete Battle Royale experience."
    font {
     bForceMonospaced=false
     bMaterialIsStencil=false
     fontMaterial=null
     fontObject=&NotoSans
     letterSpacing=0
     monospacedWidth=1
     size=24
     skewAmount=0
     typefaceFontName=NotoSans-Regular
     outlineSettings {
      bApplyOutlineToDropShadows=false
      bMiteredCorners=false
      bSeparateFillAlpha=false
      outlineMaterial=null
      outlineSize=0
      outlineColor=<rgba 0 0 0 1>
     }
    }
    shadowOffset=<0 0>
   }
   slot {class=GridSlot, properties {row=1, padding=<ltrb 0 32 0 0>}}
  }
  CTA_Button {
   widgetClassPath=/FortUILibrary/UIKit/Buttons/WBP_UIKit_Button_Loud
   bIsVariable=true
   parent=Root
   widget=CTA_Button
   properties {text=Go to lobby}
   slot {class=GridSlot, properties {horizontalAlignment=HAlign_Left, row=2, padding=<ltrb 0 64 0 0>}}
  }
 ]
}
